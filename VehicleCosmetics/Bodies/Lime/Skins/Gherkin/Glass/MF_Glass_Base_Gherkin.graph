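# Glass material function for the Gherkin skin of the Lime vehicle body.
# Builds base attributes, tinting colour and opacity by vertex-colour regions (windows, headlights, brake lights) or by dark tint with Fresnel.
# Optional static-switch stages add curvature-based thickness from normal derivatives with masked normals, and fake reflection occlusion darkening base colour and specular.
# An optional sorting fix adjusts opacity on blue-vertex areas by camera direction relative to a sorting axis.
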MaterialFunction /VehicleCosmetics/Bodies/Lime/Skins/Gherkin/Glass/MF_Glass_Base_Gherkin
graph {nodes [
  Composite() {subgraphName=Occlusion_Group}
  n1 = GetMaterialAttributes(None=NamedRerouteUsage()) {attributeGetTypes [#1]}
  Composite!(A=@n1.Opacity) {subgraphName=WindowSorting_Group}
  n2 = GetMaterialAttributes(None=NamedRerouteUsage()) {attributeGetTypes [#2, #1]}
  Composite(A=@n2.Opacity) {subgraphName=Glass_Group}
  n3 = FunctionInput(Preview=ScalarParameter() {group=1_Window, parameterName=Window Tint Amount, sortPriority=0}) {bUsePreviewValueAsDefault=true, inputName=Window Tint Amount, inputType=FunctionInput_Scalar}
  n4 = TextureSampleParameter2D() {parameterName=Glass_Sticker, texture=/Game/Environments/Sets/ArtDeco/Architecture/Textures/AD_Swatch_Grey}
  n5 = FunctionInput(Preview=ScalarParameter() {group=GLass Params, parameterName=Fresnel Inner Transparency, sortPriority=10}) {bUsePreviewValueAsDefault=true, inputName=Fresnel Inner Transparency, inputType=FunctionInput_Scalar}
  n6 = FunctionInput(Preview=ScalarParameter() {defaultValue=0.25, parameterName=Metalness?, sortPriority=15}) {bUsePreviewValueAsDefault=true, inputName=Metalness, inputType=FunctionInput_Scalar, sortPriority=15}
  n7 = FunctionInput(Preview=ScalarParameter() {defaultValue=0.125, parameterName=Roughness, sortPriority=20}) {bUsePreviewValueAsDefault=true, inputName=Roughness, inputType=FunctionInput_Scalar, sortPriority=20}
  Composite!(Window Tint Amount=@n3, ColorFront=@n4.RGB, Fresnel Inner Transparency=@n5, Metalness=@n6, Roughness=@n7) {subgraphName=BaseAttributes_Group}
  n8 = Subtract(
   A=CameraPositionWS()
   B=TransformPosition(None=VectorParameter() {parameterName=SortingCenter, defaultValue=<rgba 0 0 0 1>}.RGB) {transformType=TRANSFORMPOSSOURCE_World}
  )
  n9 = VectorParameter() {parameterName=Sorting Direction, defaultValue=<rgba 1 0 -1.1633 1>}
  n10 = LinearInterpolate(
   A=ScalarParameter() {defaultValue=1.2644, parameterName=SortingFix_1}
   B=ScalarParameter() {defaultValue=5.738, parameterName=SortingFix_2}
   Alpha=DotProduct(A=Normalize(VectorInput=@n8), B=Transform(None=@n9.RGB) {transformSourceType=TRANSFORMSOURCE_Local})
  )
  MaterialExpressionReroute_22 = Reroute(None=@n1.Opacity)
  n11 = LinearInterpolate(
   A=@MaterialExpressionReroute_22.A
   B=ScalarParameter() {defaultValue=0.2286, parameterName=SortingFixOpacity}
   Alpha=Multiply(A=VertexColor().B, B=Saturate(None=@n10))
  )
  MaterialExpressionReroute_23 = Reroute(None=@n11)
  FunctionOutput(None=StaticSwitchParameter(
    True=SetMaterialAttributes!(MaterialAttributes=@n1.MaterialAttributes, Opacity=@MaterialExpressionReroute_23.Output) {attributeSetTypes [#1]}
    False=NamedRerouteUsage()
   ) {group=ReflectionOcclusion, parameterName=UseSortingFix})
  n12 = NamedRerouteUsage()
  n13 = VertexColor()
  n14 = LinearInterpolate(
   A=LinearInterpolate(A=@n12, B=VectorParameter() {group=1_Window, parameterName=ColorSide, defaultValue=<rgba 0.0677 0.0677 0.0677 1>}.RGB, Alpha=@n13.A)
   B=VectorParameter() {group=2_HeadLights, parameterName=ColorHeadlights, defaultValue=<rgba 0.0677 0.0677 0.0677 1>}.RGB
   Alpha=@n13.R
  )
  n15 = VectorParameter() {group=3_BreakLights, parameterName=ColorBreaklights, defaultValue=<rgba 1 0 0.0158 1>}
  n16 = NamedRerouteUsage()
  n17 = StaticSwitchParameter(
   True=LinearInterpolate(
    A=LinearInterpolate(A=@n14, B=@n15.RGB, Alpha=@n13.G)
    B=VectorParameter() {group=4_Special, parameterName=ColorSpecial, defaultValue=<rgba 0.0677 0.0677 0.0677 1>}.RGB
    Alpha=@n13.B
   )
   False=LinearInterpolate(A=@n12, B=VectorParameter() {group=GLass Params, parameterName=Color_DarkTint, defaultValue=<rgba 0.001 0.001 0.001 1>}.RGB, Alpha=@n16)
  ) {group=GLass Params, parameterName=UseVertexMaskTints}
  MaterialExpressionReroute_31 = Reroute(None=@n17)
  n18 = VertexColor()
  n19 = LinearInterpolate(
   A=LinearInterpolate(
    A=ScalarParameter() {group=1_Window, parameterName=Window Tint AmountFront}
    B=ScalarParameter() {group=1_Window, parameterName=Window Tint AmountSide}
    Alpha=@n18.A
   )
   B=ScalarParameter() {group=2_HeadLights, parameterName=Window Tint Amount Headlights}
   Alpha=@n18.R
  )
  n20 = ScalarParameter() {group=3_BreakLights, parameterName=Window Tint Amount Breaklights}
  n21 = NamedRerouteUsage()
  n22 = ScalarParameter() {defaultValue=0.9, group=GLass Params, parameterName=Fresnel Inner Transparency Max Tint}
  n23 = Multiply(
   A=Fresnel(ExponentIn=ScalarParameter() {defaultValue=2, group=GLass Params, parameterName=Fresnel Exponent})
   B=ScalarParameter() {defaultValue=1, group=GLass Params, parameterName=Glass thickness}
  )
  n24 = StaticSwitchParameter(
   True=LinearInterpolate(
    A=LinearInterpolate(A=@n19, B=@n20, Alpha=@n18.G)
    B=ScalarParameter() {group=4_Special, parameterName=Window Tint Amount Special}
    Alpha=@n18.B
   )
   False=LinearInterpolate(
    A=LinearInterpolate(A=@n21, B=@n22, Alpha=@n16)
    B=ScalarParameter() {defaultValue=0.98, group=GLass Params, parameterName=Fresnel Outer Transparency}
    Alpha=Clamp(None=@n23)
   )
  ) {group=GLass Params, parameterName=UseVertexMaskTints}
  MaterialExpressionReroute_32 = Reroute(None=@n24)
  n25 = NamedRerouteUsage()
  MaterialExpressionReroute_33 = Reroute(None=@n25)
  n26 = ScalarParameter() {defaultValue=0.8, parameterName=Spec}
  MaterialExpressionReroute_34 = Reroute(None=@n26)
  n27 = NamedRerouteUsage()
  MaterialExpressionReroute_35 = Reroute(None=@n27)
  n28 = SetMaterialAttributes!(
   MaterialAttributes
   BaseColor=@MaterialExpressionReroute_31.BaseColor
   Opacity=@MaterialExpressionReroute_32.Opacity
   Roughness=@MaterialExpressionReroute_33.Roughness
   Specular=@MaterialExpressionReroute_34.Specular
   Metallic=@MaterialExpressionReroute_35.Metallic
  ) {attributeSetTypes [
    #2
    #1
    #3
    #4
    #5
   ]}
  NamedRerouteDeclaration(None=@n28) {name=BaseAttributes, nodeColor=<rgba 1 0 0.4922 1>}
  n29 = LinearInterpolate(
   A=ScalarParameter() {parameterName=CurvatureTexture_Add1}
   B=ScalarParameter() {parameterName=CurvatureTexture_Add2}
   Alpha=TextureSampleParameter2D() {parameterName=Curvature, samplerType=SAMPLERTYPE_Masks, texture=&T_BlackMask}.B
  )
  n30 = VertexNormalWS()
  n31 = DDX(Value=@n30)
  n32 = DDY(Value=@n30)
  n33 = Divide(
   A=MaterialFunctionCall(Vector 3=Add(A=@n31, B=@n32)) {materialFunction=/Engine/Functions/Engine_MaterialFunctions02/Utility/VectorLength}.V3 Length
   B=ScalarParameter() {defaultValue=0.2257, parameterName=Curvature}
  )
  n34 = ScalarParameter() {defaultValue=500, parameterName=FadeLength}
  n35 = ScalarParameter() {defaultValue=200, parameterName=FadeOffset}
  n36 = Saturate(None=LinearInterpolate(
    A=Add(A=@n29, B=@n33)
    Alpha=MaterialFunctionCall(Fade Length=@n34, Fade Offset=@n35) {materialFunction=&CameraDepthFade}.Result
   ) {constB=0})
  n37 = LinearInterpolate(A=ScalarParameter() {defaultValue=1, parameterName=BaseAdd}, B=ScalarParameter() {defaultValue=1.4, parameterName=BaseAdd2}, Alpha=@n36)
  MaterialExpressionReroute_27 = Reroute(None=@n37)
  n38 = Saturate(None=Add(A=@n2.BaseColor, B=@MaterialExpressionReroute_27.BaseColor))
  MaterialExpressionReroute_26 = Reroute(None=@n2.Opacity)
  n39 = ScalarParameter() {defaultValue=1, parameterName=OpacityAdd}
  n40 = ScalarParameter() {defaultValue=1.4, parameterName=OpacityAdd2}
  n41 = Saturate(None=Add(A=@MaterialExpressionReroute_26.A, B=LinearInterpolate(A=@n39, B=@n40, Alpha=@n36)))
  MaterialExpressionReroute_25 = Reroute(None=@n41)
  n42 = VertexColor()
  n43 = LinearInterpolate(
   A=Constant3Vector() {constant=<rgba 0 0 1 0>}
   B=TextureSampleParameter2D() {parameterName=Normal, samplerType=SAMPLERTYPE_Normal, texture=/Engine/EngineMaterials/FlatNormal}.RGB
   Alpha=Saturate(None=ChannelMaskParameter(None=@n42) {parameterName=MaskNormal}.R)
  )
  MaterialExpressionReroute_24 = Reroute(None=@n43)
  NamedRerouteDeclaration(None=StaticSwitchParameter(
    True=SetMaterialAttributes!(
     MaterialAttributes=@n2.MaterialAttributes
     BaseColor=@n38
     Opacity=@MaterialExpressionReroute_25.Opacity
     Normal=@MaterialExpressionReroute_24.Normal
    ) {attributeSetTypes [#2, #1, #6]}
    False=NamedRerouteUsage()
   ) {parameterName=UseCurvaturemap}) {name=GlassThickness, nodeColor=<rgba 1 0.2109 0 1>}
  n44 = GetMaterialAttributes(None=NamedRerouteUsage()) {attributeGetTypes [#2, #4]}
  n45 = PreSkinnedPosition()
  n46 = Divide(
   A=Add(A=ComponentMask(None=@n45) {r=true}, B=ScalarParameter() {group=ReflectionOcclusion, parameterName=OcclusionGradientAdd})
   B=ScalarParameter() {defaultValue=50, group=ReflectionOcclusion, parameterName=OcclusionGradientDistance}
  )
  n47 = VertexNormalWS()
  n48 = CameraVectorWS()
  n49 = LinearInterpolate(
   A=ScalarParameter() {defaultValue=3, group=ReflectionOcclusion, parameterName=OcclusionAngleExp1}
   B=ScalarParameter() {defaultValue=3, group=ReflectionOcclusion, parameterName=OcclusionAngleExp2}
   Alpha=Saturate(None=DotProduct(A=@n47, B=@n48))
  )
  n50 = ChannelMaskParameter(None=VertexColor()) {group=ReflectionOcclusion, maskChannel=Blue, parameterName=FakeReflectionOcclusionChannel, defaultValue=<rgba 0 0 1 0>}
  n51 = LinearInterpolate(
   B=LinearInterpolate(
    A=ScalarParameter() {group=ReflectionOcclusion, parameterName=OcclusionMin}
    B=ScalarParameter() {defaultValue=1, group=ReflectionOcclusion, parameterName=OcclusionMax}
    Alpha=VertexInterpolator(VS=@n46).PS
   )
   Alpha=Multiply(A=Saturate(None=@n49), B=@n50.R)
  ) {constA=1}
  MaterialExpressionReroute_21 = Reroute(None=@n51)
  n52 = Multiply(A=@n44.BaseColor, B=@MaterialExpressionReroute_21.Output)
  n53 = Multiply(A=@n44.Specular, B=@MaterialExpressionReroute_21.Output)
  NamedRerouteDeclaration(None=StaticSwitchParameter(
    True=SetMaterialAttributes!(MaterialAttributes=@n44.MaterialAttributes, BaseColor=@n52, Specular=@n53) {attributeSetTypes [#2, #4]}
    False=NamedRerouteUsage()
   ) {group=ReflectionOcclusion, parameterName=UseFakeReflectionOcclusion}) {name=ReflectionOcclusion, nodeColor=<rgba 0 1 0.3203 1>}
  MaterialExpressionReroute_8 = Reroute(None=@n3)
  NamedRerouteDeclaration(None=@MaterialExpressionReroute_8.Window Tint Amount) {name=NR_Window_Tint_Amount, nodeColor=<rgba 0 1 0.0859 1>}
  MaterialExpressionReroute_9 = Reroute(None=@n4.RGB)
  NamedRerouteDeclaration(None=@MaterialExpressionReroute_9.ColorFront) {name=NR_ColorFront, nodeColor=<rgba 0.8047 0 1 1>}
  MaterialExpressionReroute_7 = Reroute(None=@n5)
  NamedRerouteDeclaration(None=@MaterialExpressionReroute_7.Fresnel Inner Transparency) {name=NR_Fresnel_Inner_Transparency, nodeColor=<rgba 0.5469 1 0 1>}
  MaterialExpressionReroute_6 = Reroute(None=@n6)
  NamedRerouteDeclaration(None=@MaterialExpressionReroute_6.Metalness) {name=NR_Metaless, nodeColor=<rgba 0 1 0.4844 1>}
  MaterialExpressionReroute_5 = Reroute(None=@n7)
  NamedRerouteDeclaration(None=@MaterialExpressionReroute_5.Roughness) {name=NR_Rougness, nodeColor=<rgba 0.5703 0 1 1>}
  PinBase()
  PinBase!(Output=@n51) {reroutePins [Output {expression=&self:MaterialExpressionReroute_21}]}
  PinBase!() {reroutePins [A {expression=&self:MaterialExpressionReroute_22}]}
  PinBase!(Output=@n11) {reroutePins [Output {expression=&self:MaterialExpressionReroute_23}]}
  PinBase!() {reroutePins [A {expression=&self:MaterialExpressionReroute_26}]}
  PinBase!(BaseColor=@n37, Opacity=@n41, Normal=@n43) {reroutePins [
    BaseColor {expression=&self:MaterialExpressionReroute_27}
    Opacity {expression=&self:MaterialExpressionReroute_25}
    Normal {expression=&self:MaterialExpressionReroute_24}
   ]}
  PinBase!() {reroutePins [
    Window Tint Amount {expression=&self:MaterialExpressionReroute_8}
    ColorFront {expression=&self:MaterialExpressionReroute_9}
    Fresnel Inner Transparency {expression=&self:MaterialExpressionReroute_7}
    Metalness {expression=&self:MaterialExpressionReroute_6}
    Roughness {expression=&self:MaterialExpressionReroute_5}
   ]}
  PinBase!(BaseColor=@n17, Opacity=@n24, Roughness=@n25, Specular=@n26, Metallic=@n27) {reroutePins [
    BaseColor {expression=&self:MaterialExpressionReroute_31}
    Opacity {expression=&self:MaterialExpressionReroute_32}
    Roughness {expression=&self:MaterialExpressionReroute_33}
    Specular {expression=&self:MaterialExpressionReroute_34}
    Metallic {expression=&self:MaterialExpressionReroute_35}
   ]}
 ]}
