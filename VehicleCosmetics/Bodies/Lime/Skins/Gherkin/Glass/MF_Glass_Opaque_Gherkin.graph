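# Material function for the opaque glass on the Gherkin skin of the Lime vehicle body cosmetic.
# It wraps the shared MF_Glass_Base_Gherkin function and passes that function's result straight to its output.
# Scalar parameters feed the base function's window tint amount, Fresnel inner transparency, metalness and roughness inputs.
# A dark grey ColorFront vector parameter is also declared but is not wired to anything in this graph.
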MaterialFunction /VehicleCosmetics/Bodies/Lime/Skins/Gherkin/Glass/MF_Glass_Opaque_Gherkin
graph {nodes [
  FunctionOutput(None=MaterialFunctionCall!(
    Window Tint Amount=ScalarParameter() {defaultValue=0.7, group=1_Window, parameterName=Window Tint Amount, sortPriority=0}
    Fresnel Inner Transparency=ScalarParameter() {defaultValue=0.1, group=GLass Params, parameterName=Fresnel Inner Transparency, sortPriority=10}
    Metalness=ScalarParameter() {defaultValue=0.25, parameterName=Metalness?, sortPriority=15}
    Roughness=ScalarParameter() {parameterName=Roughness, sortPriority=20}
   ) {materialFunction=&/MF_Glass_Base_Gherkin}.Result)
  VectorParameter() {group=1_Window, parameterName=ColorFront, sortPriority=5, defaultValue=<rgba 0.0667 0.0667 0.0667 1>}
 ]}
